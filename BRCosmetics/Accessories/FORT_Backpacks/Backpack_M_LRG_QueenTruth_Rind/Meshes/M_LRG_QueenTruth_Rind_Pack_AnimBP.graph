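# Animation Blueprint for the QueenTruth Rind large backpack, adjusting its pack bone to suit the character wearing it.
# The AnimGraph chooses between two chains of additive bone offsets using IsOnOriginalCharacter, set when customization loads by matching a part name.
# Each update maps spine bone yaw to clamped backpack rotation, translation and parachuting rotation variables, though the graph shows no reads of them.
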AnimBlueprint /BRCosmetics/Accessories/FORT_Backpacks/Backpack_M_LRG_QueenTruth_Rind/Meshes/M_LRG_QueenTruth_Rind_Pack_AnimBP
target_skeleton=&/M_LRG_QueenTruth_Rind_Pack_Skeleton
graphs [
 AnimGraph {nodes [
   n1 = "Transform (Modify) Bone - Bone: pack"(ComponentPose=Local To Component(LocalPose=Input Pose() {name=InPose_0}), Translation="0,-3,0", Rotation="0,0,15") {
    rotationMode=BMM_Additive
    rotationSpace=BCS_BoneSpace
    translationMode=BMM_Additive
    translationSpace=BCS_BoneSpace
    alphaBoolBlend {blendInTime=0.1, blendOption=Linear, blendOutTime=0.1}
    boneToModify {boneName=pack}
    rotation {roll=15}
    translation {y=-3}
   }
   n2 = "Transform (Modify) Bone - Bone: pack"!(ComponentPose=@n1, bAlphaBoolEnabled:bool=Get bIsPlayingMeleeAnim(), Translation="0,0,0", Rotation="0,0,0", Scale) -> (Pose) {
    alphaInputType=Bool
    rotationMode=BMM_Additive
    translationMode=BMM_Additive
    alphaBoolBlend {blendInTime=0.2, blendOption=QuadraticInOut, blendOutTime=0.2}
    boneToModify {boneName=pack}
   }
   n3 = "Transform (Modify) Bone - Bone: pack"(ComponentPose=@n2, Translation="0,-2.5,0", Rotation="0,0,10") {
    alphaInputType=Bool
    rotationMode=BMM_Additive
    rotationSpace=BCS_BoneSpace
    translationMode=BMM_Additive
    translationSpace=BCS_BoneSpace
    alphaBoolBlend {blendInTime=0.5, blendOption=Linear, blendOutTime=0.5}
    boneToModify {boneName=pack}
    rotation {roll=10}
    translation {y=-2.5}
   }
   n4 = "Transform (Modify) Bone - Bone: pack"(ComponentPose=Local To Component(LocalPose=Input Pose()), Translation="0,-2.5,0", Rotation="0,0,10") {
    alphaInputType=Bool
    rotationMode=BMM_Additive
    rotationSpace=BCS_BoneSpace
    translationMode=BMM_Additive
    translationSpace=BCS_BoneSpace
    alphaBoolBlend {blendInTime=0.5, blendOption=Linear, blendOutTime=0.5}
    boneToModify {boneName=pack}
    rotation {roll=10}
    translation {y=-2.5}
   }
   n5 = "Transform (Modify) Bone - Bone: pack"!(ComponentPose=@n4, Alpha:float, Translation="0,0,0", Rotation="0,0,0", Scale) -> (Pose) {rotationMode=BMM_Additive, translationMode=BMM_Additive, alphaBoolBlend {blendOption=Linear}, boneToModify {boneName=pack}}
   n6 = "Transform (Modify) Bone - Bone: pack"!(ComponentPose=@n5, bAlphaBoolEnabled:bool, Translation="0,0,0", Rotation="0,0,0", Scale) -> (Pose) {
    alphaInputType=Bool
    rotationMode=BMM_Additive
    rotationSpace=BCS_BoneSpace
    translationMode=BMM_Additive
    translationSpace=BCS_BoneSpace
    alphaBoolBlend {blendInTime=0.1, blendOption=Linear, blendOutTime=0.1}
    boneToModify {boneName=pack}
   }
   Output Pose(Result=Blend Poses by bool(
     BlendPose_0=Component To Local(ComponentPose=@n3)
     BlendPose_1=Component To Local(ComponentPose=@n6)
     bActiveValue=Get IsOnOriginalCharacter() -> (IsOnOriginalCharacter:bool)
    )) {layerGroup=None}
  ]}
 EventGraph {nodes [
   Event BlueprintUpdateAnimation()
   n1 = GetOwnerSkeletalMeshComponent()
   n2 = BreakRotator(InRot=BreakTransform(InTransform=GetDeltaTransformFromRefPose(self=@n1, BoneName=spine_05, BaseName=spine_04)).Rotation)
   > Set BackPack Rotation(BackPack Rotation:Rotator=MakeRotator(Roll=MapRangeClamped(Value=@n2.Yaw, InRangeA=0, InRangeB=-30, OutRangeB=20))) -> (Output_Get:Rotator)
   > Set BackPack Translation(BackPack Translation:Vector=MakeVector(Y=MapRangeClamped(Value=@n2.Yaw, InRangeA=0, InRangeB=-30, OutRangeB=-10))) -> (Output_Get:Vector)
   n3 = GetOwnerSkeletalMeshComponent()
   n4 = BreakRotator(InRot=BreakTransform(InTransform=GetDeltaTransformFromRefPose(self=@n3, BoneName=spine_05, BaseName=spine_01)).Rotation)
   > Set BackPack Parachuting Rotation(BackPack Parachuting Rotation:Rotator=MakeRotator(Roll=MapRangeClamped(Value=@n4.Yaw, InRangeA=0, InRangeB=-30, OutRangeB=50))) -> (Output_Get:Rotator)
   n5 = Event OnCharacterCustomizationLoaded()
   n6 = Get Array Item with EFort Custom Part Type(Array=@n5.CustomMeshParts)
   > Set IsOnOriginalCharacter(IsOnOriginalCharacter:bool=Contains(SearchIn=GetDisplayName(Object=@n6), Substring=M_LRG_QueenTruth_Rind)) -> (Output_Get:bool)
  ]}
]
variables [
 BackPack Parachuting Rotation {type=Rotator}
 BackPack Rotation {type=Rotator}
 BackPack Translation {type=Vector}
 IsOnOriginalCharacter {type=boolean}
]
